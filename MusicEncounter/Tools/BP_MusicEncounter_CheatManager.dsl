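; BP_MusicEncounter_CheatManager provides a cheat function, EnableMusicEncounter, that toggles the music encounter system.
; It finds the BP_MusicEncounterManager component on the Athena game state and sets its enable flag from the parameter.
; It then uninitializes every encounter in the manager's map, whether enabling or disabling.
; Invalid game state or component references make it do nothing.
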(fn EnableMusicEncounter
  (params (Enable :type Boolean))
  (bind _returnvalue (Fortnite|Athena|GameState|GetGameStateAthena))
  (Utilities|IsValid _returnvalue
    (:Is%20Valid
      (bind _output_get (LocalVariables|SetMusic%20Encounter%20Manager%20Component (Actor|GetComponentbyClass _returnvalue "/MusicEncounter/BP_MusicEncounterManager.BP_MusicEncounterManager_C")))
      (Utilities|IsValid _output_get
        (:Is%20Valid
          (Utilities|FlowControl|Sequence
            (:then_0
              (Variables|/MusicEncounter/BP_MusicEncounterManager.BP_MusicEncounterManager_C|SetEnable%20Music%20Encounter%20System (Variables|GetEnable) (LocalVariables|GetMusic%20Encounter%20Manager%20Component)))
            (:then_1
              (bind (_keypin _valuepin) (Utilities|Map|ForEachLoop(Map) (Variables|/MusicEncounter/BP_MusicEncounterManager.BP_MusicEncounterManager_C|GetMusic%20Encounters (LocalVariables|GetMusic%20Encounter%20Manager%20Component))))
              (Class|MusicEncounterBase|UninitializeMusicEncounter _valuepin))))))))
