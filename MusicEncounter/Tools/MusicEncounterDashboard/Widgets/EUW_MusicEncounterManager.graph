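# Editor utility widget for the Music Encounter dashboard, laying out an encounter list and a details panel side by side.
# On Construct it registers an MVVM field-change delegate on the list viewmodel's SelectedEntryViewModel.
# The selection-change handler casts the entry data to BP_MusicEncounterViewModel, broadcasts it through the manager viewmodel and calls Create Interface Widgets on the details panel.
# A Create Custom State Widgets call sits unconnected in the graph.
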EditorUtilityWidgetBlueprint /MusicEncounter/Tools/MusicEncounterDashboard/Widgets/EUW_MusicEncounterManager
graphs [EventGraph {nodes [
   Selected Music Encounter Changed!() -> (OutputDelegate, then, Object:Object Reference, Field:Field Notification Id Structure) {is=title K2Node_CustomEvent}
   n1 = Get TechAudioToolsListViewModel!(self:Music Encounter List Object Reference=Get MusicEncounterList() -> (MusicEncounterList:Music Encounter List Object Reference)) -> (TechAudioToolsListViewModel:Tech Audio Tools List Viewmodel Object Reference)
   > n2 = Cast To BP_MusicEncounterViewModel!(
    execute
    Object=Get EntryData!(self:Tech Audio Tools Entry Viewmodel Object Reference=Get SelectedEntryViewModel!(self:Tech Audio Tools List Viewmodel Object Reference=@n1) -> (SelectedEntryViewModel:Tech Audio Tools Entry Viewmodel Object Reference)) -> (EntryData:Object Reference)
   ) -> (then, CastFailed, AsMusic Encounter ViewModel:Music Encounter View Model Object Reference) {is=title K2Node_DynamicCast}
   > Set with Broadcast Selected Music Encounter Viewmodel!(
    execute
    Selected Music Encounter Viewmodel:Music Encounter View Model Object Reference=@n2
    self:Music Encounter Manager Viewmodel Object Reference=Get BP_MusicEncounterManagerViewmodel() -> (BP_MusicEncounterManagerViewmodel:Music Encounter Manager Viewmodel Object Reference)
   ) -> (then, Output_Get:Music Encounter View Model Object Reference)
   > Create Interface Widgets(self:Music Encounter Details Object Reference=Get MusicEncounterDetails() -> (MusicEncounterDetails:Music Encounter Details Object Reference))
   Event Construct()
   > Add Field Value Changed Delegate!(
    execute
    self:MVVM Base Viewmodel Object Reference=Get TechAudioToolsListViewModel(self=Get MusicEncounterList())
    FieldId:Field Notification Id Structure=(FieldName="SelectedEntryViewModel")
    Delegate:Delegate=Create Event()
   ) -> (then)
   Create Custom State Widgets(self:Music Encounter Details Object Reference=Get MusicEncounterDetails())
  ]}]
variables []
widgets {
 info {rootWidgetClass=&VerticalBox, widgetCount=7}
 widgets [
  VerticalBox_30 {widgetClassPath=&VerticalBox, widget=VerticalBox_30}
  Border_0 {
   widgetClassPath=&Border
   parent=VerticalBox_30
   widget=Border_0
   properties {brushColor=<rgba 0.0103 0.0103 0.0103 1>, padding=<ltrb 0 0 0 0>}
   slot {class=VerticalBoxSlot, properties {size {sizeRule=Fill, value=1}}}
  }
  HorizontalBox_0 {
   widgetClassPath=&HorizontalBox
   parent=Border_0
   widget=HorizontalBox_0
   slot {class=BorderSlot, properties {padding=<ltrb 0 0 0 0>}}
  }
  MusicEncounterList {
   widgetClassPath=&/MusicEncounterList
   bIsVariable=true
   parent=HorizontalBox_0
   widget=MusicEncounterList
   properties {techAudioToolsListViewModel=&self:WidgetTree.MusicEncounterList.TechAudioToolsListViewModel_0}
   slot {class=HorizontalBoxSlot, properties {size {sizeRule=Fill, value=0.6}}}
  }
  Border_1 {
   widgetClassPath=&Border
   parent=HorizontalBox_0
   widget=Border_1
   properties {brushColor=<rgba 0.0176 0.0176 0.0176 1>, padding=<ltrb 0 0 0 0>}
   slot {class=HorizontalBoxSlot}
  }
  Spacer_74 {
   widgetClassPath=&Spacer
   parent=Border_1
   widget=Spacer_74
   properties {size=<2 1>}
   slot {class=BorderSlot, properties {padding=<ltrb 0 0 0 0>}}
  }
  MusicEncounterDetails {
   widgetClassPath=&/MusicEncounterDetails
   bIsVariable=true
   parent=HorizontalBox_0
   widget=MusicEncounterDetails
   slot {class=HorizontalBoxSlot, properties {size {sizeRule=Fill, value=1}}}
  }
 ]
}
