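# Editor utility dashboard for inspecting Music Encounter managers during PIE sessions, with a client selector hosting the manager widget.
# Tick polls non-dedicated-server PIE worlds for the manager component on each GameState.
# When the count grows, it builds a viewmodel per manager, updates the selector's client count and shows the first; with none, it clears them.
# Helper functions resolve viewmodels by client index and toggle between manager view and standby message.
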EditorUtilityWidgetBlueprint /MusicEncounter/Tools/MusicEncounterDashboard
graphs [
 Create Manager Viewmodels {nodes [
   Create Manager Viewmodels() {is=title K2Node_FunctionEntry}
   > n1 = For Each Loop(Array:Array of Music Encounter Manager Component Object References=Get Music Encounter Manager Components!(self) -> (Music Encounter Managers:Array of Music Encounter Manager Component Object References) {is=title K2Node_CallFunction}) -> (Array Element:Music Encounter Manager Component Object Reference)
   Set with Broadcast Music Encounter Manager Viewmodels!(
    execute=@n1.Completed
    Music Encounter Manager Viewmodels:Array of Music Encounter Manager Viewmodel Object References=Get Viewmodels() -> (Viewmodels:Array of Music Encounter Manager Viewmodel Object References)
   ) -> (then, Output_Get:Array of Music Encounter Manager Viewmodel Object References)
   n2 = Construct Object from Class(
    execute=@n1.LoopBody
    Class=/MusicEncounter/Tools/MusicEncounterDashboard/Viewmodels/BP_MusicEncounterManagerViewmodel.BP_MusicEncounterManagerViewmodel_C
   ) -> (ReturnValue:Music Encounter Manager Viewmodel Object Reference)
   > Set Music Encounter Manager!(
    execute
    self:Music Encounter Manager Viewmodel Object Reference=@n2
    Music Encounter Manager:Music Encounter Manager Component Object Reference=@n1
   ) -> (then) {is=title K2Node_CallFunction}
   > Add(
    TargetArray:Array of Music Encounter Manager Viewmodel Object References=Get Viewmodels()
    NewItem:"Music Encounter Manager Viewmodel Object Reference (by ref)"=@n2
   ) {is=title K2Node_CallArrayFunction}
  ]}
 EventGraph {nodes [
   Begin PIE!() -> (OutputDelegate, then, bIsSimulating:bool=~) {is=title K2Node_CustomEvent}
   > Bind Event to On End PIE(self:Editor Utility Subsystem Object Reference=Get EditorUtilitySubsystem(), Delegate=Create Event()) {is=title K2Node_AddDelegate}
   End PIE!() -> (OutputDelegate, then, bIsSimulating:bool=~) {is=title K2Node_CustomEvent}
   > Unbind Event from On End PIE(self:Editor Utility Subsystem Object Reference=Get EditorUtilitySubsystem(), Delegate=Create Event()) {is=title K2Node_RemoveDelegate}
   Event Destruct()
   > Unbind Event from On Begin PIE(self:Editor Utility Subsystem Object Reference=Get EditorUtilitySubsystem(), Delegate=Create Event()) {is=title K2Node_RemoveDelegate}
   > Unbind Event from On End PIE(self=Get EditorUtilitySubsystem(), Delegate=Create Event())
   Event Tick()
   n1 = Get Music Encounter Manager Components()
   > n2 = Branch(Condition="integer < integer"(A=Length(TargetArray:Array of Music Encounter Manager Component Object References=@n1), B=1))
   > Set with Broadcast Number of Clients!(
    execute
    Number of Clients:int=0
    self:BP PIEClient Selector Viewmodel Object Reference=Get BP_PIEClientSelectorViewmodel() -> (BP_PIEClientSelectorViewmodel:BP PIEClient Selector Viewmodel Object Reference)
   ) -> (then, Output_Get:int=0)
   > Clear(TargetArray:Array of Music Encounter Manager Viewmodel Object References=Get Music Encounter Manager Viewmodels() -> (Music Encounter Manager Viewmodels:Array of Music Encounter Manager Viewmodel Object References))
   > Set with Broadcast Music Encounter Viewmodels!(
    execute
    Music Encounter Viewmodels:Array of Music Encounter View Model Object References=Make Array!() -> (Array:Array of Music Encounter View Model Object References)
    self:Music Encounter Manager Viewmodel Object Reference=Get BP_MusicEncounterManagerViewmodel!(self:Music Encounter Manager Widget Object Reference=Get EUW_MusicEncounterManager() -> (EUW_MusicEncounterManager:Music Encounter Manager Widget Object Reference)) -> (BP_MusicEncounterManagerViewmodel:Music Encounter Manager Viewmodel Object Reference)
   ) -> (then, Output_Get:Array of Music Encounter View Model Object References)
   n3 = Get Music Encounter Manager Components()
   n4 = Get BP_PIEClientSelectorViewmodel()
   Branch(
    execute=@n2.else
    Condition="integer > integer"(A=Length(TargetArray=@n3), B=Get Number of Clients!(self:BP PIEClient Selector Viewmodel Object Reference=@n4) -> (Number of Clients:int=0))
   )
   > Set with Broadcast Number of Clients(Number of Clients=Length(TargetArray=Get Music Encounter Manager Components()), self=Get BP_PIEClientSelectorViewmodel()) -> (Output_Get=0)
   > Create Manager Viewmodels() {is=title K2Node_CallFunction}
   > SetBP_MusicEncounterManagerViewmodel!(
    execute
    self:Music Encounter Manager Widget Object Reference=Get EUW_MusicEncounterManager()
    Viewmodel:Music Encounter Manager Viewmodel Object Reference="Get (a copy)"(Array:Array of Music Encounter Manager Viewmodel Object References=Get Music Encounter Manager Viewmodels()) -> (Output:Music Encounter Manager Viewmodel Object Reference)
   ) -> (then)
   Event Construct()
   > Bind Event to On Begin PIE(self:Editor Utility Subsystem Object Reference=Get EditorUtilitySubsystem(), Delegate=Create Event()) {is=title K2Node_AddDelegate}
   > Bind Event to On End PIE(self=Get EditorUtilitySubsystem(), Delegate=Create Event())
  ]}
 Get Manager Widget Visibility {nodes [
   n1 = Get Manager Widget Visibility!() -> (then, Number of Managers:int) {is=title K2Node_FunctionEntry}
   > Return Node!(execute, Visibility:ESlateVisibility Enum=Select(Option 0=Collapsed, Option 1=Visible, Index="integer > integer"(A=@n1)))
  ]}
 Get Music Encounter Manager Components {nodes [
   Get Music Encounter Manager Components() {is=title K2Node_FunctionEntry}
   > n1 = GetPIEWorlds!(execute, bIncludeDedicatedServer:bool=false) -> (then, ReturnValue:Array of World Object References)
   > n2 = For Each Loop(Array:Array of World Object References=@n1) -> (Array Element:World Object Reference)
   Return Node!(
    execute=@n2.Completed
    Music Encounter Managers:Array of Music Encounter Manager Component Object References=Get Music Managers() -> (Music Managers:Array of Music Encounter Manager Component Object References)
   )
   n3 = GetComponentByClass!(
    self:Actor Object Reference=GetGameStateAthenaFromContext!(WorldContextObject:Object Reference=@n2) -> (ReturnValue:Fort Game State Athena Object Reference)
    ComponentClass:Actor Component Class Reference=/MusicEncounter/BP_MusicEncounterManager.BP_MusicEncounterManager_C
   ) -> (ReturnValue:Music Encounter Manager Component Object Reference)
   > Is Valid(InputObject=@n3)
   > Add Unique(
    TargetArray:Array of Music Encounter Manager Component Object References=Get Music Managers()
    NewItem:"Music Encounter Manager Component Object Reference (by ref)"=@n3
   )
  ]}
 Get Music Encounter Manager from Selected Client {nodes [
   n1 = Get Music Encounter Manager from Selected Client!() -> (then, Selected Client Index:int) {is=title K2Node_FunctionEntry}
   n2 = Get Music Encounter Manager Viewmodels()
   > Return Node!(
    execute
    Selected Manager Viewmodel:Music Encounter Manager Viewmodel Object Reference=Select(
     Option 0:Music Encounter Manager Viewmodel Object Reference
     Option 1:Music Encounter Manager Viewmodel Object Reference="Get (a copy)"(Array=@n2, Dimension 1=@n1)
     Index=Is Valid Index!(TargetArray:Array of Music Encounter Manager Viewmodel Object References=@n2, IndexToTest:int=@n1) -> (ReturnValue:bool)
    ) -> (ReturnValue:Music Encounter Manager Viewmodel Object Reference)
   )
  ]}
 Get Standby Widget Visibility {nodes [
   n1 = Get Standby Widget Visibility!() -> (then, Number of Managers:int) {is=title K2Node_FunctionEntry}
   > Return Node!(
    execute
    Visibility:ESlateVisibility Enum=Select(Option 0:ESlateVisibility Enum=Collapsed, Option 1:ESlateVisibility Enum=Visible, Index="integer <= integer"(A=@n1)) -> (ReturnValue:ESlateVisibility Enum)
   )
  ]}
]
variables [Music Encounter Manager Viewmodels {type="array<BP_MusicEncounterManagerViewmodel_C>", default []}]
widgets {
 info {namedSlotCount=1, rootWidgetClass=&Border, widgetCount=6}
 widgets [
  Border_106 {widgetClassPath=&Border, widget=Border_106, properties {brushColor=<rgba 0.0176 0.0176 0.0176 1>, padding=<ltrb 2 2 2 2>}}
  VerticalBox_0 {
   widgetClassPath=&VerticalBox
   parent=Border_106
   widget=VerticalBox_0
   slot {class=BorderSlot, properties {padding=<ltrb 2 2 2 2>}}
  }
  EUW_PIEClientSelector {
   widgetClassPath=/FortTechAudioTools/Widgets/PIEClientSelector/EUW_PIEClientSelector
   bIsVariable=true
   parent=VerticalBox_0
   widget=EUW_PIEClientSelector
   slot {class=VerticalBoxSlot, properties {size {sizeRule=Fill, value=1}}}
  }
  EUW_MusicEncounterManager {
   widgetClassPath=/MusicEncounter/Tools/MusicEncounterDashboard/Widgets/EUW_MusicEncounterManager
   bIsVariable=true
   namedSlotHost=EUW_PIEClientSelector
   widget=EUW_MusicEncounterManager
  }
  Border_104 {
   widgetClassPath=&Border
   parent=VerticalBox_0
   widget=Border_104
   properties {horizontalAlignment=HAlign_Center, verticalAlignment=VAlign_Center, brushColor=<rgba 0.0075 0.0075 0.0075 1>, padding=<ltrb 0 0 0 0>}
   slot {class=VerticalBoxSlot, properties {size {sizeRule=Fill, value=1}}}
  }
  TextBlock_79 {
   widgetClassPath=&TextBlock
   parent=Border_104
   widget=TextBlock_79
   properties {
    text=Waiting for PIE session...
    colorAndOpacity {colorUseRule=UseColor_Specified, specifiedColor=<rgba 0.5271 0.5271 0.5271 1>}
    font {
     bForceMonospaced=false
     bMaterialIsStencil=false
     fontMaterial=null
     fontObject=&Roboto
     letterSpacing=0
     monospacedWidth=1
     size=9
     skewAmount=0
     typefaceFontName=Italic
     outlineSettings {
      bApplyOutlineToDropShadows=false
      bMiteredCorners=false
      bSeparateFillAlpha=false
      outlineMaterial=null
      outlineSize=0
      outlineColor=<rgba 0 0 0 1>
     }
    }
   }
   slot {class=BorderSlot, properties {horizontalAlignment=HAlign_Center, verticalAlignment=VAlign_Center, padding=<ltrb 0 0 0 0>}}
  }
 ]
}
