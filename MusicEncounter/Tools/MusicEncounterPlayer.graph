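# Editor Utility Widget for previewing music encounters in the editor, with class picker, Start/Pause buttons, Phase view and details panel.
# Construct spawns a transient editor-world actor carrying a Music Encounter Manager component.
# Spawn Component adds a mock component using that manager and the picked class, then sends its request.
# Set Music Encounter event initializes the given instance and feeds the Phase view and view model; picker changes or Destruct uninitialize it.
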EditorUtilityWidgetBlueprint /MusicEncounter/Tools/MusicEncounterPlayer
graphs [EventGraph {nodes [
   "On Property Changed (MusicEncounterPicker)"!() -> (OutputDelegate, then, PropertyName:Name=None) {is=title K2Node_ComponentBoundEvent}
   > n1 = Get!(execute:Exec) -> (then:Exec, else:Exec, Music Encounter:Music Encounter Base Object Reference)
   > Uninitialize Music Encounter(self:Music Encounter Base Object Reference=@n1)
   "On Clicked (StartMusicEncounter_Button)"() {is=title K2Node_ComponentBoundEvent}
   > n2 = Get()
   > Start Music Encounter(self:Music Encounter Base Object Reference=@n2)
   "On Clicked (PauseMusicEncounter_Button)"() {is=title K2Node_ComponentBoundEvent}
   > n3 = Get()
   > Pause Music Encounter(self:Music Encounter Base Object Reference=@n3)
   "On Clicked (PauseMusicEncounter_Button_1)"() {is=title K2Node_ComponentBoundEvent}
   n4 = Get Editor World Context Actor() -> (Editor World Context Actor:Actor Object Reference)
   > n5 = Add Component by Class!(
    execute:Exec
    self:Actor Object Reference=@n4
    WorldContextObject:Object Reference=@n4
    Class:Actor Component Class Reference=/MusicEncounter/Tools/MusicEncounterDashboard/BP_MockMusicEncounterComponent.BP_MockMusicEncounterComponent_C
    Mock Music Encounter Manager:Music Encounter Manager Component Object Reference=Get Music Encounter Manager() -> (Music Encounter Manager:Music Encounter Manager Component Object Reference)
    Music Encounter Class:Music Encounter Base Soft Class Reference=Get Music Encounter Class() -> (Music Encounter Class:Music Encounter Base Soft Class Reference)
    Music Encounter Instance Tag:Gameplay Tag Structure
    Component Tag:Gameplay Tag Structure
    Required Initialization Tags:Gameplay Tag Container Structure
    Initialization Blocking Tags:Gameplay Tag Container Structure
   ) -> (then:Exec, ReturnValue:BP Mock Music Encounter Component Object Reference) {is=title K2Node_AddComponentByClass}
   > Send Request(self:BP Mock Music Encounter Component Object Reference=@n5)
   n6 = Set Music Encounter!() -> (OutputDelegate, then, Music Encounter:Music Encounter Base Object Reference) {is=title K2Node_CustomEvent}
   > Set Music Encounter(Music Encounter:Music Encounter Base Object Reference=@n6.Music Encounter) -> (Output_Get:Music Encounter Base Object Reference) {is=title K2Node_VariableSet}
   > SetObject(
    self=Get MusicEncounterPhase_PropertyView() -> (MusicEncounterPhase_PropertyView:Single Property View Object Reference)
    NewObject=Get Music Encounter() -> (Music Encounter:Music Encounter Base Object Reference)
   )
   > Initialize Music Encounter(self:Music Encounter Base Object Reference=Get Music Encounter())
   > Set Music Encounter!(
    execute
    self:Music Encounter View Model Object Reference=Get BP_MusicEncounterViewModel() -> (BP_MusicEncounterViewModel:Music Encounter View Model Object Reference)
    Music Encounter:Music Encounter Base Object Reference=Get Music Encounter()
   ) -> (then) {is=title K2Node_CallFunction}
   Event PreConstruct()
   > SetObject(self=Get MusicEncounterPicker() -> (MusicEncounterPicker:Single Property View Object Reference), NewObject=Self-Reference())
   Event Destruct()
   > n7 = Get()
   > Uninitialize Music Encounter(self=@n7)
   Event Construct()
   > n8 = GetEditorWorld(self=Get UnrealEditorSubsystem())
   > n9 = SpawnTransientActor!(
    execute
    WorldContextObject:Object Reference=@n8
    ActorClass:Actor Class Reference=&Actor
    SpawnTransform_Location:Vector="0, 0, 0"
    SpawnTransform_Rotation:Rotator="0, 0, 0"
    SpawnTransform_Scale:Vector="1,1,1"
    CollisionHandlingOverride:ESpawnActorCollisionHandlingMethod Enum=Undefined
    OwnerActor:Actor Object Reference
    InstigatorPawn:Pawn Object Reference
   ) -> (then, SpawnedActor:Actor Object Reference)
   > n10 = Set Editor World Context Actor(Editor World Context Actor:Actor Object Reference=@n9) -> (Output_Get:Actor Object Reference)
   > n11 = Add Component by Class!(execute, self=@n10, WorldContextObject=@n10, Class=/MusicEncounter/BP_MusicEncounterManager.BP_MusicEncounterManager_C) -> (then, ReturnValue:Music Encounter Manager Component Object Reference)
   > Set Music Encounter Manager(Music Encounter Manager:Music Encounter Manager Component Object Reference=@n11) -> (Output_Get:Music Encounter Manager Component Object Reference)
  ]}]
variables [
 Editor World Context Actor {type=Actor, default=null}
 Music Encounter {type=BP_MusicEncounter_C, default=null}
 Music Encounter Class {type=Class, default=null}
 Music Encounter Manager {type=BP_MusicEncounterManager_C, default=null}
]
widgets {
 info {rootWidgetClass=&HorizontalBox, widgetCount=11}
 widgets [
  HorizontalBox_0 {widgetClassPath=&HorizontalBox, widget=HorizontalBox_0}
  VerticalBox_0 {
   widgetClassPath=&VerticalBox
   parent=HorizontalBox_0
   widget=VerticalBox_0
   slot {class=HorizontalBoxSlot, properties {size {sizeRule=Fill, value=0.4}}}
  }
  MusicEncounterPicker {
   widgetClassPath=&SinglePropertyView
   parent=VerticalBox_0
   widget=MusicEncounterPicker
   properties {propertyName=Music Encounter Class}
   slot {class=VerticalBoxSlot}
  }
  StartMusicEncounter_Button {widgetClassPath=&EditorUtilityButton, parent=VerticalBox_0, widget=StartMusicEncounter_Button, slot {class=VerticalBoxSlot}}
  TextBlock_34 {
   widgetClassPath=&TextBlock
   parent=StartMusicEncounter_Button
   widget=TextBlock_34
   properties {
    text=Start Music Encounter
    font {
     bForceMonospaced=false
     bMaterialIsStencil=false
     fontMaterial=null
     fontObject=&Roboto
     letterSpacing=0
     monospacedWidth=1
     size=8.25
     skewAmount=0
     typefaceFontName=Regular
     outlineSettings {
      bApplyOutlineToDropShadows=false
      bMiteredCorners=false
      bSeparateFillAlpha=false
      outlineMaterial=null
      outlineSize=0
      outlineColor=<rgba 0 0 0 1>
     }
    }
   }
   slot {class=ButtonSlot}
  }
  PauseMusicEncounter_Button {widgetClassPath=&EditorUtilityButton, parent=VerticalBox_0, widget=PauseMusicEncounter_Button, slot {class=VerticalBoxSlot}}
  TextBlock {
   widgetClassPath=&TextBlock
   parent=PauseMusicEncounter_Button
   widget=TextBlock
   properties {
    text=Pause Music Encounter
    font {
     bForceMonospaced=false
     bMaterialIsStencil=false
     fontMaterial=null
     fontObject=&Roboto
     letterSpacing=0
     monospacedWidth=1
     size=8.25
     skewAmount=0
     typefaceFontName=Regular
     outlineSettings {
      bApplyOutlineToDropShadows=false
      bMiteredCorners=false
      bSeparateFillAlpha=false
      outlineMaterial=null
      outlineSize=0
      outlineColor=<rgba 0 0 0 1>
     }
    }
   }
   slot {class=ButtonSlot}
  }
  PauseMusicEncounter_Button_1 {
   widgetClassPath=&EditorUtilityButton
   parent=VerticalBox_0
   widget=PauseMusicEncounter_Button_1
   slot {class=VerticalBoxSlot}
  }
  TextBlock_1 {
   widgetClassPath=&TextBlock
   parent=PauseMusicEncounter_Button_1
   widget=TextBlock_1
   properties {
    text=Spawn Component
    font {
     bForceMonospaced=false
     bMaterialIsStencil=false
     fontMaterial=null
     fontObject=&Roboto
     letterSpacing=0
     monospacedWidth=1
     size=8.25
     skewAmount=0
     typefaceFontName=Regular
     outlineSettings {
      bApplyOutlineToDropShadows=false
      bMiteredCorners=false
      bSeparateFillAlpha=false
      outlineMaterial=null
      outlineSize=0
      outlineColor=<rgba 0 0 0 1>
     }
    }
   }
   slot {class=ButtonSlot}
  }
  MusicEncounterPhase_PropertyView {
   widgetClassPath=&SinglePropertyView
   parent=VerticalBox_0
   widget=MusicEncounterPhase_PropertyView
   properties {propertyName=Phase}
   slot {class=VerticalBoxSlot}
  }
  MusicEncounterDetails {
   widgetClassPath=/MusicEncounter/Tools/MusicEncounterDashboard/Widgets/MusicEncounterDetails
   bIsVariable=true
   parent=HorizontalBox_0
   widget=MusicEncounterDetails
   slot {class=HorizontalBoxSlot, properties {size {sizeRule=Fill, value=0.6}}}
  }
 ]
}
